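; Creative-mode physics ball that players can push or pickaxe, with scaling of its visible mesh to match physics bounds.
; Server hits knock pawns back by ball speed and optionally kill them; client overlaps apply predicted impulses.
; A cosmetic timer feeds smoothed speed to rolling audio, alongside collision sounds and destruction effects guarded against same-tick replay spawns.
; It also provides async material swaps, a lazily created energy shell and a floating HUD indicator.
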(event Custom|Play%20Rolling%20Sound
  (Variables|Default|SetVelocity%20Smoothed (Math|Interpolation|FInterpTo (Variables|Default|GetVelocity%20Smoothed) (Math|Vector|VectorLength (Transformation|GetVelocity)) (Utilities|Time|GetWorldDeltaSeconds) 1e+01))
  (Audio|Parameter|SetFloatParameter (Variables|Default|GetBall_Roll_Loop_Cue) "Speed" (Math|Float|MapRangeClamped (Variables|Default|GetVelocity%20Smoothed) 0.0 1e+03 0.05 1.3) :!class /Script/AudioExtensions.AudioParameterControllerInterface))

(event Cosmetic|EventOnEnteredWaterVolume
  (params (WaterSurfacePoint :type Vector%20%28by%20ref%29 :default "0, 0, 0")))

(event EventBeginPlay
  (bind (_initialvelocity _overridemassinkg _overrideangulardampening _overridelineardampening _playerforcemultiplier _pawnknockbackmultiplier) (Utilities|Struct|BreakPhysicsPawnObjectInitialParameters (Variables|FortPhysicsPawnObject|GetInitialParameters)))
  (CallFunction|AdjustVisibleMeshScaletoMatchPhysicsMesh)
  (Variables|Default|SetPawnKnockbackMultipler _pawnknockbackmultiplier)
  (Networking|CanExecuteCosmeticEvents
    (:True
      (Game|BindEventtoOnDestroyed (delegate Custom|PlayDestructionEffects))
      (Utilities|Time|SetTimerbyEvent (EventDispatchers|CreateEvent :!event Custom|Play%20Rolling%20Sound) 0.1 true)
      (Transformation|SetAbsolute (Variables|Default|GetValueBar_Root) false true false) ; Use Absolute rotation for the Value bar to ensure\nit stays above the centre when the object rotates\nrather than rotating with it.
    )
    (:False
      (Components|DestroyComponent (Variables|Default|GetCreative_ValueBar)) ; If you can't execute cosmetic events, \nyou don't need this component.
    )))

(event Custom|PlayDestructionEffects
  (params (DestroyedActor :type Actor%20Object%20Reference))
  (bind _returnvalue (Transformation|GetActorLocation))
  (if (CallFunction|ShouldPlayDeathEffects)
    (Audio|PlaySoundatLocation (Variables|Default|GetDestructionSound) _returnvalue) ; play hologram sound
    (Effects|SpawnEmitteratLocation :EmitterTemplate (Variables|Default|GetDestructionParticles) :Location _returnvalue :Scale (Transformation|GetActorScale3D) :bAutoDestroy true :PoolingMethod "None" :bAutoActivateSystem true)))

(event EventOnCollisionHitServer
  (params (HitLocation :type Vector%20%28by%20ref%29 :default "0, 0, 0") (VelocityOfHitter :type Vector%20%28by%20ref%29 :default "0, 0, 0") (HitNormal :type Vector%20%28by%20ref%29 :default "0, 0, 0") (HitActor :type Actor%20Object%20Reference) (HitSurfaceType :type EPhysicalSurface%20Enum :default "SurfaceType_Default"))
  (bind _asplayer_pawn_athena (Utilities|Casting|CastToPlayerPawn_Athena HitActor))
  (CallFunction|AttemptImpulseHitter _asplayer_pawn_athena VelocityOfHitter HitNormal)
  (if (Variables|Default|GetKillPawnOnHit)
    (FortPawn|ForceKill _asplayer_pawn_athena "(TagName=\"Gameplay.Effect.InstantDeath.Environment\")" 0 self)))

(event OnComponentBeginOverlap(Sphere)
  (params (OverlappedComponent :type Primitive%20Component%20Object%20Reference) (OtherActor :type Actor%20Object%20Reference) (OtherComp :type Primitive%20Component%20Object%20Reference) (OtherBodyIndex :type Integer) (bFromSweep :type Boolean) (SweepResult :type Hit%20Result%20Structure%20%28by%20ref%29))
  (bind _returnvalue (Transformation|GetVelocity OtherActor))
  (Utilities|FlowControl|Sequence
    (:then_0
      (bind _asfort_player_pawn (Utilities|Casting|CastToFortPlayerPawn OtherActor))
      (Networking|SwitchHasAuthority
        (:Remote
          (CallFunction|ApplyPredictedImpulsefromPawnHitClient _returnvalue (Math|Vector|Normalize (- (Transformation|GetActorLocation) (Transformation|GetActorLocation _asfort_player_pawn))) _asfort_player_pawn))))
    (:then_1
      (Networking|SwitchHasAuthority
        (:Remote
          (Audio|OnCollision (Variables|Default|GetFortCollisionAudio) (Transformation|GetActorLocation) _returnvalue))))))

(event Collision|EventHit
  (params (MyComp :type Primitive%20Component%20Object%20Reference) (Other :type Actor%20Object%20Reference) (OtherComp :type Primitive%20Component%20Object%20Reference) (bSelfMoved :type Boolean) (HitLocation :type Vector :default "0, 0, 0") (HitNormal :type Vector :default "0, 0, 0") (NormalImpulse :type Vector :default "0, 0, 0") (Hit :type Hit%20Result%20Structure%20%28by%20ref%29))
  (bind (_bblockinghit _binitialoverlap _time _distance _location _impactpoint _normal _impactnormal) (Collision|BreakHitResult Hit))
  (Audio|OnCollision (Variables|Default|GetFortCollisionAudio) (Transformation|GetActorLocation) (* _impactnormal (Transformation|GetVelocity))))

(event EventOnDamaged
  (params (Damage :type Float%20%28single-precision%29) (DamageTags :type Gameplay%20Tag%20Container%20Structure%20%28by%20ref%29) (EffectContext :type Gameplay%20Effect%20Context%20Handle%20Structure%20%28by%20ref%29) (EventInstigator :type Controller%20Object%20Reference) (DamageCauser :type Actor%20Object%20Reference))
  ;[0.07245 1 0.1129] OnDamaged - Detect if the damage was coming from a player's pickaxe
  (|CallParentFunction Damage DamageTags EffectContext EventInstigator DamageCauser)
  (if (GameplayTags|HasTag DamageTags "(TagName=\"Weapon.Melee.Impact.Pickaxe\")")
    (CallFunction|OnHitwithPixkaxe EventInstigator)))

(event Custom|OnHitWithPixkaxe
  (params (HitInstigator :type Controller%20Object%20Reference))
  (bind (_initialparameters_initialvelocity _initialparameters_overridemassinkg _initialparameters_overrideangulardampening _initialparameters_overridelineardampening _initialparameters_playerforcemultiplier) (Utilities|Struct|BreakPhysicsPawnObjectInitialParameters (Variables|GetInitialParameters_InitialVelocity)))
  ;[0.6545 1 0.2493] Apply impulse when hit with pickaxe (by default, pickaxe does not apply impulse to anything that it hits)
  (Utilities|IsValid HitInstigator
    (:Is%20Valid
      (PhysicsObject|ApplyImpulse HitInstigator (* (Math|Vector|GetForwardVector (Pawn|GetControlRotation HitInstigator :!class /Script/Engine.Controller)) (* (Variables|Default|GetBasePickaxeImpulseStrength) _initialparameters_playerforcemultiplier))))))

(event Custom|ChangeBallMaterial
  (params (MaterialIndex :type Integer))
  (bind _materialtoset (Variables|Default|GetMaterialToSet))
  ;[1 1 1] Update The Material on the Ball
  (if (Utilities|Array|IsValidIndex _materialtoset MaterialIndex)
    (bind _object (Utilities|AsyncLoadAsset (Utilities|Array|Get(acopy) _materialtoset MaterialIndex))
      (:Completed
        (bind _asmaterial_interface (Utilities|Casting|CastToMaterialInterface _object))
        (Rendering|Material|SetMaterial (Variables|StaticMeshActor|GetVisibleMeshComponent) 0 _asmaterial_interface)))))

(event Custom|ApplyEnergyShell
  (params (ApplyEnergyShell :type Boolean))
  (bind _energyshellmesh (Variables|Default|GetEnergyShellMesh))
  (bind _energyshellmesh_2 (Variables|Default|GetEnergyShellMesh))
  ;[1 1 1] Enable/Disable Energy Shield
  (if ApplyEnergyShell
    (Utilities|IsValid _energyshellmesh_2
      (:Is%20Valid
        (Rendering|SetVisibility _energyshellmesh_2 true false))
      (:Is%20Not%20Valid
        (bind _returnvalue (AddComponent|NODE_AddStaticMeshComponent-1))
        (bind _output_get (Variables|Default|SetEnergyShellMesh _returnvalue))
        (Transformation|AttachComponentToComponent _output_get (Variables|StaticMeshActor|GetVisibleMeshComponent) "None" "KeepRelative" "KeepRelative" "KeepRelative" true)
        (Rendering|SetVisibility (Variables|Default|GetEnergyShellMesh) true false)))
    (else
      (Utilities|IsValid _energyshellmesh
        (:Is%20Valid
          (Rendering|SetVisibility _energyshellmesh false false))))))

(fn AdjustVisibleMeshScaleToMatchPhysicsMesh
  (bind (_origin _boxextent _sphereradius) (Collision|GetComponentBounds (Variables|StaticMeshActor|GetPhysicsMeshComponent)))
  (bind (_origin_2 _boxextent_2 _sphereradius_2) (Collision|GetComponentBounds (Variables|StaticMeshActor|GetVisibleMeshComponent)))
  ;[1 1 1] Rp / Rv = ScaleRation * CurrentScale
  ;
  ; This adjusts the visible static mesh so that it closely matches the static mesh of the physics object
  (Transformation|SetRelativeScale3D (Variables|StaticMeshActor|GetVisibleMeshComponent) (* (Variables|/Script/Engine.SceneComponent|GetRelativeScale3D (Variables|StaticMeshActor|GetVisibleMeshComponent)) (/ _sphereradius _sphereradius_2))))

(fn AttemptImpulseHitter
  (params (HittingPawn :type Actor%20Object%20Reference) (VelocityOfHitter :type Vector) (HitNormal :type Vector))
  ;[1 0.4389 0.01696] Getting the impulse to the player pawn is quite difficult and will require more tuning than currently worth. The mass and speed ranges available cause this to be harder than preciously thought
  (Physics|GetPhysicsLinearVelocity (Variables|StaticMeshActor|GetPhysicsMeshComponent) "None")
  (bind _asfort_pawn (Utilities|Casting|CastToFortPawn HittingPawn))
  (Pawn|Character|LaunchCharacterJump _asfort_pawn (* (+ (Math|Vector|NegateVector HitNormal) "0,0.000000,0.200000") (Math|Float|float*float (Variables|Default|GetMaxPawnKnockbackForce) (Variables|Default|GetPawnKnockbackMultipler) (Math|Float|Max(Float) (Variables|PhysicsObject|GetCachedSpeed) 2e+01))) true true true false))

(fn CalculateImpulse
  (params (ActorCausingHit :type Creative%20Ball%20Object%20Reference) (HitNormal :type Vector))
  (bind _physicsmeshcomponent (Variables|/Script/FortniteGame.FortPhysicsPawnObject|GetPhysicsMeshComponent ActorCausingHit))
  (bind _returnvalue (Physics|GetMass _physicsmeshcomponent))
  (bind _returnvalue_2 (* _returnvalue 0.01667))
  (bind _returnvalue_3 (PhysicsObject|GetPreviousLinearAcceleration ActorCausingHit))
  (bind _returnvalue_4 (* _returnvalue_3 _returnvalue_2))
  (bind _returnvalue_5 (* _returnvalue_4 HitNormal))
  (Development|PrintString (Utilities|String|ToString(Vector) _returnvalue_5) true true "(R=1.000000,G=0.354451,B=0.000000,A=1.000000)" 6e+01)
  (return _returnvalue_5))

(fn HideHUD
  (bind _creative_valuebar (Variables|Default|GetCreative_ValueBar)
    (:then
      (Class|CreativeValueBarComponent|HideValueBar _creative_valuebar))
    (:else)))

(fn ShouldPlayDeathEffects
  ;[0.5 0.25 0.25] Replays create and destroy some actors in the same tick
  (return (not (Math|Float|NearlyEqual(Float) (Actor|GetGameTimeSinceCreation) 0.0 0.01))))

(fn SpawnorUpdateHUD
  (params (ShowDistance :type Float%20%28double-precision%29) (Indicator%20Index :type Integer) (Require%20LOS :type Boolean) (Friendly%20Teams :type Array%20of%20Integers) (Hostile%20Teams :type Array%20of%20Integers) (ShowOnlyToTeamIndex :type Integer) (TextFriendly :type String) (TextNeutral :type String) (TextHostile :type String) (TextScale :type Float%20%28double-precision%29) (DisplayDistanceText :type Boolean) (ClampProjectionToViewport :type Boolean) (DisplayOffscreenArrow :type Boolean) (Indicator%20Icon :type Texture%202D%20Soft%20Object%20Reference))
  (Networking|CanExecuteCosmeticEvents
    (:True
      (Class|CreativeValueBarComponent|SpawnOrUpdateValueBar (Variables|Default|GetCreative_ValueBar) (Variables|StaticMeshActor|GetVisibleMeshComponent) self "/Game/Creative/Devices/Common/Components/Creative_FloatingValueBar.Creative_FloatingValueBar_C" 0 ShowDistance 0 -1.0 1e+02 Require%20LOS Friendly%20Teams Hostile%20Teams ShowOnlyToTeamIndex false TextFriendly TextNeutral TextHostile TextScale false DisplayDistanceText ClampProjectionToViewport DisplayOffscreenArrow Indicator%20Icon true "(X=1.000000,Y=1.000000)" false false))))

(fn ConstructionScript
  (CallFunction|AdjustVisibleMeshScaletoMatchPhysicsMesh))
