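; Creative ball spawner device that spawns a physics pawn ball above its base, with type-based mass and damping.
; Authority spawns it on minigame start or pre-game, removes it on end, and timer-respawns destroyed balls during games.
; Option changes respawn it on non-cosmetic instances, while cosmetic ones update material, energy shell and team HUD, also via SpawnedBall RepNotify.
; The construction script previews the scaled ball in editor.
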(event Custom|RespawnBallOnDelay
  (params (DestroyedActor :type Actor%20Object%20Reference))
  (if (and (>= (Variables|ToyOptions|GetRespawnDelay) 0.0) (Variables|Default|GetbMinigameInProgress))
    (bind _returnvalue (Utilities|Time|SetTimerbyEvent (delegate Custom|SpawnBallAfterdelay) (Variables|ToyOptions|GetRespawnDelay)))
    (Variables|Default|SetRespawnBallTimer _returnvalue)
    (else
      (CallFunction|SpawnBall))))

(event Custom|SpawnBallAfterdelay
  (Utilities|Time|ClearandInvalidateTimerbyHandle (Variables|Default|GetRespawnBallTimer))
  (if (Variables|Default|GetbMinigameInProgress)
    (CallFunction|SpawnBall)))

(event EventBeginPlay
  (bind _fortminigameprogress (Variables|Default|GetFortMinigameProgress))
  (Minigame|AssignMinigameStartedDelegate _fortminigameprogress (EventDispatchers|CreateEvent :!event Custom|OnMinigameStarted)) ; oing it this way instead of using the events because you can't call parent on a component event, but you can on BeginPlay
  (Minigame|BindEventtoMinigameEndedDelegate _fortminigameprogress (EventDispatchers|CreateEvent :!event Custom|OnMinigameEnded))
  (Visibility|SetVisibilityComponents (Variables|Default|GetVisibleInGameComponent) (Utilities|Array|MakeArray (Variables|Default|GetSpawnerBase)) true))

(event Custom|OnMinigameStarted
  (Variables|Default|SetbMinigameInProgress true)
  (Networking|SwitchHasAuthority
    (:Authority
      (CallFunction|SpawnBall))))

(event Custom|OnMinigameEnded
  (Variables|Default|SetbMinigameInProgress false)
  (Networking|SwitchHasAuthority
    (:Authority
      (Utilities|Time|ClearandInvalidateTimerbyHandle (Variables|Default|GetRespawnBallTimer))
      (CallFunction|UnbindandDestroyPreviousBall))))

(event MinigameStateChangedDelegate(FortMinigameProgress)
  (params (Minigame :type Fort%20Minigame%20Object%20Reference) (NewMinigameState :type EFortMinigameState%20Enum :default "PreGame"))
  ;[1 1 1] Respawn the ball each time the game enters pre-game.
  (Networking|SwitchHasAuthority
    (:Authority
      (if (Utilities|Enum|Equal(Enum) NewMinigameState "PreGame")
        (CallFunction|SpawnBall)))))

(event Building|EventOnIsForPreviewingChangedBP
  (params (bNewIsForPreviewing :type Boolean))
  ;[1 1 1] Respawn the ball when any property changes
  (Networking|CanExecuteCosmeticEvents
    (:True
      (CallFunction|UpdateSpawnedBallVisuals) ; After the visuals settings are replicated to the clients\nUpdate the visuals
    )
    (:False
      (if (Building|IsPreviewBuildingActor)
        (CallFunction|UnbindandDestroyPreviousBall)
        (else
          (CallFunction|SpawnBall) ; Respawn Ball with the new options
        )))))

(event OnAnyPropertyChanged(ToyOptionsComponent)
  ;[1 1 1] Respawn the ball when any property changes
  (Networking|CanExecuteCosmeticEvents
    (:True
      (CallFunction|UpdateSpawnedBallVisuals) ; After the visuals settings are replicated to the clients\nUpdate the visuals
    )
    (:False
      (if (Building|IsPreviewBuildingActor)
        (CallFunction|UnbindandDestroyPreviousBall)
        (else
          (CallFunction|SpawnBall) ; Respawn Ball with the new options
        )))))

(event OnPropertiesLoaded(ToyOptionsComponent)
  ;[1 1 1] Respawn the ball when any property changes
  (Networking|CanExecuteCosmeticEvents
    (:True
      (CallFunction|UpdateSpawnedBallVisuals) ; After the visuals settings are replicated to the clients\nUpdate the visuals
    )
    (:False
      (if (Building|IsPreviewBuildingActor)
        (CallFunction|UnbindandDestroyPreviousBall)
        (else
          (CallFunction|SpawnBall) ; Respawn Ball with the new options
        )))))

(event Custom|SetEditorMaterial
  ;[1 1 1] For Setting the Material in editor only
  (bind _loaded_object (Creative|Loading|AsyncLoadEditorSafe (Utilities|Array|Get(acopy) (Variables|Default|GetMaterialToSet) (Variables|ToyOptions|GetMaterialIndex)))
    (:Loaded
      (bind _asmaterial_interface (Utilities|Casting|CastToMaterialInterface _loaded_object))
      (Rendering|Material|SetMaterial (Variables|Default|GetEditorBallMesh) 0 _asmaterial_interface))))

(event EventDestroyed
  (CallFunction|UnbindandDestroyPreviousBall))

(event Custom|EnableHUD
  (params (Instigator :type Actor%20Object%20Reference))
  ;[0.7501 0 1] Receivers
  (Variables|Default|SetShow%20HUD true))

(event Custom|DisableHUD
  (params (Instigator :type Actor%20Object%20Reference))
  ;[0.7501 0 1] Receivers
  (Variables|Default|SetShow%20HUD false))

(fn DisableSimulation
  (Variables|ToyOptions|SetbSimulateBall false)
  (CallFunction|ToggleSimulationEffectsOfBallSpawner false)
  (Networking|SwitchHasAuthority
    (:Authority
      (OptionData|SetPropertyOverrides (Variables|Default|GetToyOptionsComponent) (Utilities|Map|MakeMap "bSimulateBall" "0")))))

(fn GetBoundsComponents
  (params (SelfActor :type Actor%20Object%20Reference))
  (return (Utilities|Array|MakeArray (Variables|Default|GetSpawnerBase))))

(fn GetPropertiesBasedOnBallType
  (params (BallType :type Integer))
  (switch int BallType
    (:Default
      (Development|PrintString "WARNING: Non-Valid Ball Option" true true "(R=1.000000,G=0.448992,B=0.000000,A=1.000000)" 4.0))
    (:0
      ;[1 1 1] Default
      (return 4e+02 0.8 0.4))
    (:1
      ;[1 1 1] Light
      (return 1e+02 0.9 0.8))
    (:2
      ;[1 1 1] Heavy
      (return 1e+03 0.8 0.4))
    (:3
      ;[1 1 1] Slick
      (return 3e+02 0.1 0.05))
    (:4
      ;[1 1 1] Sporty (Ideal Rocket League Settings)
      (return 2e+02 0.5 0.2))))

(fn OnRep_bSimulateBall
  (Development|SetHiddeninGame (Variables|Default|GetSpawnerBase) false true)
  (CallFunction|ToggleSimulationEffectsOfBallSpawner (Variables|ToyOptions|GetbSimulateBall)))

(fn OnRep_ShowHUD
  (Networking|CanExecuteCosmeticEvents
    (:True
      (bind _spawnedball (Variables|Default|GetSpawnedBall)
        (:then
          (if (Variables|Default|GetShow%20HUD)
            (CallFunction|SpawnOrUpdateHUD)
            (else
              (Class|CreativeBall|HideHUD _spawnedball))))
        (:else)))))

(fn OnRep_SpawnedBall
  (Networking|CanExecuteCosmeticEvents
    (:True
      (Variables|Default|GetSpawnedBall
        (:then
          (CallFunction|UpdateSpawnedBallVisuals)
          (Audio|PlaySoundatLocation (Variables|Default|GetStartSimulationSound) (Transformation|GetActorLocation)))
        (:else)))))

(fn SpawnorUpdateHUD
  (Networking|CanExecuteCosmeticEvents
    (:True
      (LocalVariables|SetlocalHostileTeams (Knobs|Team|GetTeamArray self))
      (Utilities|FlowControl|Sequence
        (:then_0
          ;[1 1 1] Parse Team Data
          (if (Knobs|Team|IsTeamNeutral (Variables|ToyOptions|GetFriendlyTeamIndex))
            (Utilities|Array|Clear (LocalVariables|GetlocalHostileTeams))
            (elif (Knobs|Team|IsTeamAny (Variables|ToyOptions|GetFriendlyTeamIndex))
              (LocalVariables|SetlocalFriendlyTeam (LocalVariables|GetlocalHostileTeams)) ; Friendly teams take priority over hostile teams.
              (elif (Knobs|Team|IsValidTeamIndex (Variables|ToyOptions|GetFriendlyTeamIndex))
                (Utilities|Array|Add (LocalVariables|GetlocalFriendlyTeam) (Knobs|Team|GetTeamValueasIndex (Variables|ToyOptions|GetFriendlyTeamIndex)))))))
        (:then_1
          (if (Math|Conversions|ToBoolean(Integer) (Variables|ToyOptions|GetHealthBarIndex))
            (bind _spawnedball (Variables|Default|GetSpawnedBall)
              (:then
                (Class|CreativeBall|SpawnOrUpdateHUD _spawnedball (Variables|ToyOptions|GetHealthBarDistance) 0 (Variables|ToyOptions|GetbHealthBarRequireLOS) (LocalVariables|GetlocalFriendlyTeam) (LocalVariables|GetlocalHostileTeams) (Variables|ToyOptions|GetShowTeamIndex) (Variables|ToyOptions|GetHealthBarOverrideTextFriendly) (Variables|ToyOptions|GetHealthBarOverrideTextNeutral) (Variables|ToyOptions|GetHealthBarOverrideTextEnemy) (Variables|ToyOptions|GetHealthBarTextSize) (Math|Conversions|ToBoolean(Integer) (Variables|ToyOptions|GetDisplayDistanceText)) (Math|Conversions|ToBoolean(Integer) (Variables|ToyOptions|GetClampToScreen)) (Math|Conversions|ToBoolean(Integer) (Variables|ToyOptions|GetShowOffscreenArrow)) (Creative|UI|IconPicker|GetIconLibraryTexture (Variables|ToyOptions|GetHealthBarIdentifierIndex))))
              (:else))
            (else
              (bind _spawnedball_2 (Variables|Default|GetSpawnedBall)
                (:then
                  (Class|CreativeBall|HideHUD _spawnedball_2))
                (:else)))))))))

(fn SpawnBall
  (bind (_mass _angulardampening _lineardampening) (CallFunction|GetPropertiesBasedonBallType (Variables|ToyOptions|GetBallType)))
  (CallFunction|UnbindandDestroyPreviousBall)
  (bind _returnvalue (Class|FortPhysicsPawnObject|SpawnPhysicsPawnObject "/Game/Creative/Devices/RigidBodies/CreativeBall.CreativeBall_C" (Math|Transform|MakeTransform (+ (Transformation|GetActorLocation) (Math|Vector|RotateVector (Math|Vector|MakeVector 0.0 0.0 (* (Variables|Default|GetBallRadiusOffsetWhenScaleIsOne) (Variables|ToyOptions|GetInitialScale))) (Transformation|GetActorRotation))) "0, 0, 0" (Math|Vector|MakeVector (Variables|ToyOptions|GetInitialScale) (Variables|ToyOptions|GetInitialScale) (Variables|ToyOptions|GetInitialScale))) (Utilities|Struct|MakePhysicsPawnObjectInitialParameters (Variables|ToyOptions|GetInitialVelocity) _mass _angulardampening _lineardampening (Variables|ToyOptions|GetPlayerForceMultipler) (Variables|ToyOptions|GetPawnImpulseMultipler) (Variables|ToyOptions|GetbEnableGravity) (Variables|ToyOptions|GetbEnableGravityOnHit))))
  (bind _output_get (Variables|Default|SetSpawnedBall _returnvalue))
  (Utilities|IsValid _output_get
    (:Is%20Valid
      (Game|BindEventtoOnDestroyed (Variables|Default|GetSpawnedBall) (EventDispatchers|CreateEvent :!event RespawnBallOnDelay))
      (Variables|/Game/Creative/Devices/RigidBodies/CreativeBall.CreativeBall_C|SetKillPawnOnHit (and (Variables|ToyOptions|GetbShouldKillPawnOnHit) (Variables|Default|GetbMinigameInProgress)) (Variables|Default|GetSpawnedBall)))))

(fn ToggleSimulationEffectsOfBallSpawner
  (params (bSimulationEnabled :type Boolean))
  (bind _spawnerbase (Variables|Default|GetSpawnerBase))
  (bind _spawnerbase_2 (Variables|Default|GetSpawnerBase))
  (if bSimulationEnabled
    (Collision|SetCollisionResponsetoChannel _spawnerbase_2 "ECC_Pawn" "ECR_Ignore")
    (Collision|SetCollisionResponsetoChannel _spawnerbase_2 "ECC_Vehicle" "ECR_Ignore")
    (Networking|CanExecuteCosmeticEvents
      (:True
        (Audio|PlaySoundatLocation (Variables|Default|GetStartSimulationSound) (Transformation|GetActorLocation))))
    (else
      (Actor|DestroyActor (Variables|Default|GetSpawnedBall))
      (Collision|SetCollisionResponsetoChannel _spawnerbase "ECC_Pawn" "ECR_Block")
      (Collision|SetCollisionResponsetoChannel _spawnerbase "ECC_Vehicle" "ECR_Block"))))

(fn UnbindAndDestroyPreviousBall
  (bind _spawnedball (Variables|Default|GetSpawnedBall)
    (:then
      (Game|UnbindEventfromOnDestroyed _spawnedball (EventDispatchers|CreateEvent :!event RespawnBallOnDelay))
      (Actor|DestroyActor _spawnedball))
    (:else)))

(fn UpdateSpawnedBallVisuals
  (Networking|CanExecuteCosmeticEvents
    (:True
      (bind _spawnedball (Variables|Default|GetSpawnedBall)
        (:then
          (CallFunction|SpawnOrUpdateHUD) ; Do this here as well as on the Spawn Ball Repnotify to avoid a race condition on initial spawn where the ball manages to spawn and replicate before the Toy Component Properties are loaded.
          (Class|CreativeBall|ChangeBallMaterial _spawnedball (Variables|ToyOptions|GetMaterialIndex))
          (Class|CreativeBall|ApplyEnergyShell _spawnedball (Variables|Default|GetUseEnergyShell)))
        (:else)))))

(fn ConstructionScript
  (bind _editorballmesh (Variables|Default|GetEditorBallMesh))
  (|CallParentFunction)
  (Collision|SetCollisionEnabled (Variables|Default|GetSpawnerBase) "QueryAndPhysics")
  (if (Editor|InEditor)
    (Rendering|SetVisibility (Variables|Default|GetEditorBallMesh) true false)
    (Transformation|SetWorldScale3D _editorballmesh (Math|Vector|MakeVector (Variables|ToyOptions|GetInitialScale) (Variables|ToyOptions|GetInitialScale) (Variables|ToyOptions|GetInitialScale)))
    (Transformation|SetWorldLocation _editorballmesh (+ (Transformation|GetActorLocation) (Math|Vector|RotateVector (Math|Vector|MakeVector 0.0 0.0 (* (Variables|Default|GetBallRadiusOffsetWhenScaleIsOne) (Variables|ToyOptions|GetInitialScale))) (Transformation|GetActorRotation))) false false)
    (Rendering|SetVisibility (Variables|Default|GetEnergyShell) (Variables|Default|GetUseEnergyShell) false)
    (CallFunction|SetEditorMaterial)))
